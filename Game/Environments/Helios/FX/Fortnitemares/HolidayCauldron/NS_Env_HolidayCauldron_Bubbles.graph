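# Niagara system for purple bubbles that swell and pop on the surface of the Fortnitemares holiday cauldron.
# One emitter steadily spawns scaled mesh bubbles in a small area, with size and dynamic material curves driven over each particle's life.
# The same particle also drives four sprite renderers through custom splash, wave and two ripple attributes, timed by curves to appear late in life as the bubble bursts.
# Ripple colors are tinted by time of day.
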NiagaraSystem /Game/Environments/Helios/FX/Fortnitemares/HolidayCauldron/NS_Env_HolidayCauldron_Bubbles
dependencies {
 usedDataInterfaces [&NiagaraDataInterfaceCurve, &NiagaraDataInterfaceColorCurve]
 usedDynamicInputs []
 usedModules [&SystemState]
}
emitters {Bubbles {
  emitterUpdateScript [
   &EmitterState()
   &SpawnRate(SpawnRate=3.2, Use Spawn Probability=false)
  ]
  particleSpawnScript [
   &InitializeParticle(
    Color Maximum=<rgba 0.2591 0.053 0.6458 1>
    Color Minimum=<rgba 0.376 0.1235 0.849 1>
    Color Mode=::NewEnumerator2
    Lifetime Min=1.5
    Lifetime Mode=::NewEnumerator1
    Mesh Scale Mode=::NewEnumerator1
    Mesh Uniform Scale Max=0.08
    Mesh Uniform Scale Min=0.045
    Sprite Rotation Mode=::NewEnumerator2
    UsePositionOffset=false
    Write Material Random=true
   )
   &ShapeLocation(Disc Coverage=0.2024, MaskSpawnGroup=false, Override Seed=false, Ring Radius=30, Shape Primitive=::NewEnumerator4, UseOwnerScale=true)
  ]
  particleUpdateScript [
   &ParticleState()
   &ScaleMeshSize(Scale Factor=&VectorFromFloat(Value=&FloatFromCurve(FloatCurve=DataInterface() {propertyValues {Curve {keys [<at 0 0.2>,<at 0.8 1>,<at 1 0>]}}})))
   &DynamicMaterialParameters(
    Index 0 Param 1=&FloatFromCurve(FloatCurve=DataInterface() {propertyValues {Curve {keys [
         <at 0 0>
         {interpMode=RCIM_Cubic, leaveTangent=1.2072, tangentMode=RCTM_Break, time=0.55, value=0}
         {arriveTangent=3.283, tangentMode=RCTM_User, time=1, value=1}
        ]}}})
    Index 0 Param 4=1.1
    Module.Index0Param0Write=true
    Module.Index0Param1Write=true
    Module.Index0Param2Write=true
    Module.Index0Param3Write=true
   )
   &SolveForcesAndVelocity(Clamp Velocity=false, Limit Acceleration=false)
   /SetVariables(
    Particles.SplashDynamicParamenter:Vector4f=&MakeVector4(
     X=&FloatFromCurve(FloatCurve=DataInterface() {propertyValues {Curve {keys [
          <at 0 1>
          {interpMode=RCIM_Cubic, leaveTangent=3.8935, tangentMode=RCTM_Break, time=0.8, value=1.5}
          {arriveTangent=171.8661, tangentMode=RCTM_User, time=1, value=17}
         ]}}})
     Y=0.7
     Z=0
     W=0
    )
    Particles.SplashIndex:NiagaraFloat=&RandomRangeFloat(Minimum=0, Maximum=4)
    Particles.SplashSize:Vector2f=&Vector2DFromFloat(Value=&FloatFromCurve(
      FloatCurve=DataInterface() {propertyValues {Curve {keys [<at 0 0>,<at 0.7 0>,<at 0.85 1>]}}}
      Scale Curve=&RandomRangeFloat(Minimum=30, Maximum=35)
     ))
   ) {isSetParameters=true}
   /SetVariables(
    Particles.WaveColor:LinearColor=&ColorFromCurve(ColorCurve=DataInterface() {propertyValues {
       AlphaCurve {keys [<at 0.6765 0>,<at 0.7462 1>,<at 0.9338 0 RCIM_Cubic>]}
       BlueCurve {keys [<at 0.8093 1 RCIM_Cubic>,<at 1 0.3542 RCIM_Cubic>]}
       GreenCurve {keys [<at 0.8093 0.3304 RCIM_Cubic>,<at 1 0.039 RCIM_Cubic>]}
       RedCurve {keys [<at 0.8093 0.6054 RCIM_Cubic>,<at 1 0.0886 RCIM_Cubic>]}
      }})
    Particles.WaveSize:Vector2f=&Vector2DFromFloat(Value=&MakeFloatFromVector2d(Vector 2D=&Vector2DFromFloat(Value=&FloatFromCurve(
        FloatCurve=DataInterface() {propertyValues {Curve {keys [<at 0 0>,<at 0.7 0>,<at 0.8 1>]}}}
        Scale Curve=&RandomRangeFloat(Minimum=25, Maximum=30)
       ))))
   ) {isSetParameters=true}
   /SetVariables(
    Particles.RippleColor:LinearColor=&MultiplyLinearColor(
     Linear Color=&ColorFromCurve(ColorCurve=DataInterface() {propertyValues {
        AlphaCurve {keys [<at 0.6765 0>,<at 0.7462 0.2>,<at 0.9056 0 RCIM_Cubic>]}
        BlueCurve {keys [<at 0.8093 1 RCIM_Cubic>,<at 1 0.3542 RCIM_Cubic>]}
        GreenCurve {keys [<at 0.8093 0.6537 RCIM_Cubic>,<at 1 0.0115 RCIM_Cubic>]}
        RedCurve {keys [<at 0.8093 0.787 RCIM_Cubic>,<at 1 0.1009 RCIM_Cubic>]}
       }})
     Scale Factor=&MakeVector4_FromLinearColor(LinearColor=&FNTimeOfDayColor(Night Color=<rgba 0.556 0.5008 1 1>, Morning Color=<rgba 1 0.6695 0.6095 1>, Day Color=<rgba 1 1 1 1>, Evening Color=<rgba 1 1 1 1>))
    )
    Particles.RippleFacing:Vector3f=<0 0 1>
    Particles.RippleSize:Vector2f=&Vector2DFromFloat(Value=&FloatFromCurve(FloatCurve=DataInterface() {propertyValues {Curve {keys [<at 0.5 0>,<at 1 1>]}}}, Scale Curve=35))
   ) {isSetParameters=true}
   /SetVariables(
    Particles.Ripple2Color:LinearColor=&MultiplyLinearColor(
     Linear Color=&ColorFromCurve(ColorCurve=DataInterface() {propertyValues {
        AlphaCurve {keys [<at 0.8143 0>,<at 0.8876 0.1>,<at 0.9661 0 RCIM_Cubic>]}
        BlueCurve {keys [<at 0.8093 1 RCIM_Cubic>,<at 1 0.3542 RCIM_Cubic>]}
        GreenCurve {keys [<at 0.8093 0.5627 RCIM_Cubic>,<at 1 0.0705 RCIM_Cubic>]}
        RedCurve {keys [<at 0.8093 0.6911 RCIM_Cubic>,<at 1 0.1519 RCIM_Cubic>]}
       }})
     Scale Factor=&MakeVector4_FromLinearColor(LinearColor=&FNTimeOfDayColor(Night Color=<rgba 0.556 0.5008 1 1>, Morning Color=<rgba 1 0.6695 0.6095 1>, Day Color=<rgba 1 1 1 1>, Evening Color=<rgba 1 1 1 1>))
    )
    Particles.Ripple2Size:Vector2f=&Vector2DFromFloat(Value=&FloatFromCurve(
      FloatCurve=DataInterface() {propertyValues {Curve {keys [
          {interpMode=RCIM_Cubic, leaveTangent=0.6678, tangentMode=RCTM_Break, time=0.6, value=0}
          {arriveTangent=2.5069, tangentMode=RCTM_User, time=1, value=0.65}
         ]}}}
      Scale Curve=30
     ))
   ) {isSetParameters=true}
  ]
  rendererClasses [&NiagaraMeshRendererProperties, &NiagaraSpriteRendererProperties]
  renderers [
   {rendererClass=&NiagaraMeshRendererProperties, rendererIndex=0}
   {rendererClass=&NiagaraSpriteRendererProperties, rendererIndex=1}
   {rendererClass=&NiagaraSpriteRendererProperties, rendererIndex=2}
   {rendererClass=&NiagaraSpriteRendererProperties, rendererIndex=3}
   {rendererClass=&NiagaraSpriteRendererProperties, rendererIndex=4}
  ]
 }}
summary {
 systemName=NS_Env_HolidayCauldron_Bubbles
 emitters [{emitterName=Bubbles, rendererClasses [&NiagaraMeshRendererProperties, &NiagaraSpriteRendererProperties]}]
 userVariables []
}
